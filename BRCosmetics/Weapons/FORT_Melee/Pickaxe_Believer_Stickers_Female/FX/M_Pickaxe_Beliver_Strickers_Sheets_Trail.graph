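# Translucent Niagara ribbon material for the sticker-sheet trail on the female Believer Stickers pickaxe.
# A tiled sine on UV1's V axis, offset by dynamic Width and sharpened, forms repeating segments tinted by particle color for emissive.
# Use BlackLine blends opacity toward a segment offset by twice Width, allowing dark borders; BlackLineWidth is unread.
# Opacity also gets a sine falloff across U, an optional V fade via dynamic Mask, and particle alpha.
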Material /BRCosmetics/Weapons/FORT_Melee/Pickaxe_Believer_Stickers_Female/FX/M_Pickaxe_Beliver_Strickers_Sheets_Trail
graph {
 properties {bUsedWithNiagaraRibbons=true, blendMode=BLEND_Translucent, pixelDepthOffsetMode=PDOM_Legacy, usedShadingModels=Default Lit}
 nodes [
  n1 = ParticleColor()
  n2 = ComponentMask(None=TextureCoordinate() {coordinateIndex=1, vTiling=10}) {g=true}
  n3 = Sine(None=Add(A=Multiply(A=@n2) {constB=0.5}) {constB=-0.25})
  n4 = Add(A=@n3, B=DynamicParameter() {defaultValue=<rgba 1 0.3 0.2 1>, paramNames [Mask, Width, BlackLineWidth, Param4]}.Width) {constB=0.3}
  n5 = Saturate(None=Multiply(A=Saturate(None=@n4)) {constB=10})
  MaterialFunctionCall(Emissive=Multiply(A=@n1.RGB, B=@n5)) {materialFunction=&MF_FX_EmissiveAdjustments}
  n6 = DynamicParameter() {defaultValue=<rgba 1 0.3 0.2 1>, paramNames [Mask, Width, BlackLineWidth, Param4]}
  n7 = Add(A=@n6.Width, B=@n6.Width)
  n8 = Multiply(A=Saturate(None=Add(A=@n3, B=@n7) {constB=0.3})) {constB=10}
  n9 = ComponentMask(None=TextureCoordinate()) {r=true}
  n10 = Power(Base=Sine(None=Multiply(A=@n9) {constB=0.5})) {constExponent=3}
  n11 = Multiply(
   A=LinearInterpolate(A=@n5, B=Saturate(None=@n8), Alpha=ScalarParameter() {parameterName=Use BlackLine})
   B=Saturate(None=Multiply(A=@n10) {constB=10})
  )
  n12 = OneMinus(None=ComponentMask(None=TextureCoordinate()) {g=true})
  n13 = Multiply(A=@n11, B=Saturate(None=Multiply(A=@n12) {constB=2}))
  n14 = DynamicParameter() {defaultValue=<rgba 1 0.3 0.2 1>, paramNames [Mask, Width, BlackLineWidth, Param4]}
  Multiply(A=@n1.A, B=Saturate(None=LinearInterpolate(A=@n11, B=@n13, Alpha=@n14.Mask)))
 ]
}
